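# Animation Blueprint for the pet backpack cosmetic, adding secondary physics motion to its dangling parts.
# It takes an input pose and runs four chained Anim Dynamics nodes on the right zipper, left zipper, roll and pan bones before output.
# Each simulates in actor space with overridden angular damping and per-bone angular limits; the roll is only partially blended.
# The event graph holds no logic.
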AnimBlueprint /BRCosmetics/Accessories/FORT_Backpacks/BackPack_Pet/Meshes/BackPack_Pet_AnimBP
target_skeleton=&/BackPack_Pet_Skeleton
graphs [
 AnimGraph {nodes [
   AnimGraphNode_SubInput_0 = Input Pose()
   n1 = Anim Dynamics - Bone: dyn_zipper_right(ComponentPose=Local To Component(LocalPose=@AnimGraphNode_SubInput_0), Alpha=1, ExternalForce="0,0,0") {
    angularDampingOverride=0.9
    bGravityOverrideInSimSpace=true
    bOverrideAngularDamping=true
    bUsePlanarLimit=false
    simulationSpace=Actor
    alphaBoolBlend {blendOption=Linear}
    boundBone {boneName=dyn_zipper_right}
    physicsBodyDefinitions [{
      sphereCollisionRadius=0
      boundBone {boneName=dyn_zipper_right}
      boxExtents=<5 5 2.5>
      constraintSetup {angularTargetAxis=AxisZ, twistAxis=AxisZ, angularLimitsMax=<15 15 5>, angularLimitsMin=<-15 -15 -5>, angularTarget=<0 0 0>}
      localJointOffset=<0 0 -0.1>
     }]
   }
   n2 = Anim Dynamics - Bone: dyn_zipper_left(ComponentPose=@n1, Alpha=1, ExternalForce="0,0,0") {
    angularDampingOverride=0.9
    bGravityOverrideInSimSpace=true
    bOverrideAngularDamping=true
    bUsePlanarLimit=false
    simulationSpace=Actor
    alphaBoolBlend {blendOption=Linear}
    boundBone {boneName=dyn_zipper_left}
    physicsBodyDefinitions [{
      sphereCollisionRadius=0
      boundBone {boneName=dyn_zipper_left}
      boxExtents=<5 5 2.5>
      constraintSetup {angularTargetAxis=AxisZ, twistAxis=AxisZ, angularLimitsMax=<15 15 5>, angularLimitsMin=<-15 -15 -5>, angularTarget=<0 0 0>}
      localJointOffset=<0 0 -0.1>
     }]
   }
   n3 = Anim Dynamics - Bone: dyn_roll(ComponentPose=@n2, Alpha=0.7, ExternalForce="0,0,0") {
    angularDampingOverride=0.98
    bGravityOverrideInSimSpace=true
    bOverrideAngularDamping=true
    bUsePlanarLimit=false
    simulationSpace=Actor
    alphaBoolBlend {blendOption=Linear}
    boundBone {boneName=dyn_roll}
    physicsBodyDefinitions [{
      sphereCollisionRadius=0
      boundBone {boneName=dyn_roll}
      boxExtents=<7 7 3>
      constraintSetup {angularTargetAxis=AxisZ, twistAxis=AxisZ, angularLimitsMax=<0 5 5>, angularLimitsMin=<0 -5 -5>, angularTarget=<0 0 0>}
      localJointOffset=<0 0 -0.1>
     }]
   }
   n4 = Anim Dynamics - Bone: dyn_pan(ComponentPose=@n3, Alpha=1, ExternalForce="0,0,0") {
    angularDampingOverride=0.9
    bGravityOverrideInSimSpace=true
    bOverrideAngularDamping=true
    bUsePlanarLimit=false
    simulationSpace=Actor
    alphaBoolBlend {blendOption=Linear}
    boundBone {boneName=dyn_pan}
    physicsBodyDefinitions [{
      sphereCollisionRadius=0
      boundBone {boneName=dyn_pan}
      boxExtents=<5 5 2.5>
      constraintSetup {angularTargetAxis=AxisZ, twistAxis=AxisZ, angularLimitsMax=<0 25 15>, angularLimitsMin=<-40 -25 -15>, angularTarget=<0 0 0>}
      localJointOffset=<0 0 -0.1>
     }]
   }
   Output Pose(Result=Component To Local(ComponentPose=@n4)) {layerGroup=None}
  ]}
 EventGraph {nodes [Event BlueprintUpdateAnimation()]}
]
variables []
